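# Additive, unlit, two-sided Niagara ribbon material for the lightning trail of the SatireCane Ode pickaxe.
# A UV-controlled lightning texture's red channel blends emissive from a blue C3 parameter toward particle color.
# Red plus boosted green form the opacity mask, shaped by particle alpha, tiling clamp and fades at both U ends.
# Depth fade is optional and off by default, and near-camera fade is bypassed on mobile.
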Material /BRCosmetics/Weapons/FORT_Melee/Pickaxe_SatireCane_Ode/FX/M_LightningRibbon_SatireCane_Ode_Trail
graph {
 properties {
  bAutomaticallySetUsageInEditor=false
  bUsedWithNiagaraRibbons=true
  blendMode=BLEND_Additive
  pixelDepthOffsetMode=PDOM_Legacy
  shadingModel=MSM_Unlit
  twoSided=true
  usedShadingModels=Unlit
 }
 nodes [
  n1 = ParticleColor()
  n2 = TextureObjectParameter() {
   group=NoiseBMask
   parameterName=NoiseBMask_Texture
   sortPriority=0
   texture=/BRCosmetics/Gadgets/Assets/VinderTech_GliderChute/Glider_BentBaton/FX/Materials/T_Glider_BentBaton_Lightning_Blur
  }
  n3 = VectorParameter() {group=NoiseBMask, parameterName=NoiseBMask_UVControl, channelNames {a=VPan, b=UPan, g=VScale, r=UScale}, defaultValue=<rgba 1 1 0 0>}
  n4 = MaterialFunctionCall!(
   Texture=@n2
   AdditionalUVOffset
   UVControl_RGB=@n3
   UVControl_A=@n3
   "[Override]TimeInput"
   "[Override]ScaleUVByCenter"=StaticBool() {value=true}
   "[Override]UVInput"=TextureCoordinate()
  ) {materialFunction=/Game/Effects/Niagara/MasterMaterials/Functions/Common/FXCommon_TextureSample_V2}
  n5 = Multiply(A=Power(Base=Add(A=@n4.R, B=@n4.R), Exp=ScalarParameter() {defaultValue=2, parameterName=Power})) {constB=2.5}
  MaterialFunctionCall(Emissive=LinearInterpolate(A=VectorParameter() {parameterName=C3, defaultValue=<rgba 0 0.1044 0.75 1>}.RGB, B=@n1.RGB, Alpha=Saturate(None=@n5))) {materialFunction=&MF_FX_EmissiveAdjustments}
  n6 = Add(A=Multiply(A=@n4.R), B=Saturate(None=Multiply(A=@n4.G) {constB=10}))
  n7 = MaterialFunctionCall!(TexCoord=@n4.UV, ClampChannel=Constant2Vector() {r=1}) {materialFunction=/Game/Effects/Fort_Effects/MaterialFunctions/MF_FX_TilingMethod}
  n8 = TextureCoordinate()
  n9 = Power(Base=OneMinus(None=ComponentMask(None=@n8) {r=true}), Exp=ScalarParameter() {defaultValue=2.5, parameterName=Beginning Fade Exp})
  n10 = Multiply(A=@n9, B=ScalarParameter() {defaultValue=1, parameterName=Power_mask_intersity})
  n11 = Multiply(
   A=Multiply(A=@n1.A, B=Multiply(A=@n6, B=@n7.Result))
   B=LinearInterpolate(A=@n9, B=Saturate(None=@n10), Alpha=ScalarParameter() {parameterName=Use_Clearer_power_mask})
  )
  n12 = DepthFade(Opacity=@n11, FadeDistance=ScalarParameter() {defaultValue=5, group=Depth Fade, parameterName=NearFadeDistance})
  n13 = ScalarParameter() {group=Depth Fade, parameterName=Use Depth Fade}
  n14 = MaterialFunctionCall(Distance=ScalarParameter() {defaultValue=128, parameterName=CameraFadeDistance}) {materialFunction=&Near_Camera_Fade}
  n15 = Constant() {r=1}
  n16 = TextureCoordinate()
  n17 = Multiply(A=Power(Base=ComponentMask(None=@n16) {r=true}), B=ScalarParameter() {defaultValue=25, parameterName=left edge fade intensity}) {constB=5}
  Multiply(
   A=Multiply(
    A=LinearInterpolate(A=@n11, B=@n12, Alpha=@n13)
    B=ShadingPathSwitch(Default=@n14.Result, Mobile=@n15) {desc=Migrated from FeatureLevelSwitch}
   )
   B=LinearInterpolate(A=Constant() {r=1}, B=Saturate(None=@n17), Alpha=ScalarParameter() {defaultValue=1, parameterName=Use left edge fade})
  )
 ]
}
