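# Niagara swing trail for the SatireCane_Ode pickaxe, drawing green ribbons and sparks along the weapon's skeletal mesh.
# Lightning, Trail_FaceCamera and Trail_Soft_Dark are AnimTrail ribbons following mesh bones or sockets.
# Sparks spawn per distance moved at a socket, then drift with curl noise and drag.
# User.SwingSectionIndex selects system-age alpha fade curves for the ribbons and a camera-facing tip offset, while time-of-day scales the dark trail's color.
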NiagaraSystem /BRCosmetics/Weapons/FORT_Melee/Pickaxe_SatireCane_Ode/FX/NS_Pickaxe_SatireCane_Ode_Trail
dependencies {
 usedDataInterfaces [&NiagaraDataInterfaceCurve, &NiagaraDataInterfaceColorCurve]
 usedDynamicInputs [&FNTimeOfDayScalar]
 usedModules [
  &SystemState
  &self:SystemSpawnScript.SystemScriptSource.SystemScriptGraph.NiagaraNodeAssignment_1.SetVariables_#1
  &self:SystemSpawnScript.SystemScriptSource.SystemScriptGraph.NiagaraNodeAssignment_2.SetVariables_#2
 ]
}
emitters {
 Lightning {
  emitterUpdateScript [
   &EmitterState()
   &FN_SpawnAnimTrail()
   /SetVariables(Emitter.rand:NiagaraFloat=&RandomRangeFloat(Minimum=1, Maximum=1, Evaluation Type=::NewEnumerator1)) {isSetParameters=true}
  ]
  particleSpawnScript [
   &FN_InitializeAnimTrailBoneIndecies(First Bone or Socket Index=1, Second Bone or Socket Index=0)
   &FN_InitializeAnimTrail(AnimTrailWidth=0.75, Color=<rgba 0.5 10 3.1917 1>, Lifetime=0.27)
  ]
  particleUpdateScript [
   &FN_UpdateAnimTrail(SkeletalMesh=$User.SkeletalMesh)
   &ParticleState()
   &ScaleColor(
    Scale Alpha=&Multiply_Float(
     A=&MergeHelper_FloatSwitch(
      0=&FloatFromCurve(
       FloatCurve=DataInterface() {propertyValues {Curve {
          defaultValue=1
          keys [
           {arriveTangent=0.074, interpMode=RCIM_Cubic, leaveTangent=0.074, tangentMode=RCTM_User, time=0, value=1}
           {arriveTangent=-8.6402, interpMode=RCIM_Cubic, leaveTangent=-8.6402, tangentMode=RCTM_User, time=0.1972, value=0.5421}
           {arriveTangent=0.0063, interpMode=RCIM_Cubic, leaveTangent=0.0063, tangentMode=RCTM_User, time=0.26, value=0}
          ]
         }}}
       CurveIndex=$Engine.System.Age
      )
      1=&FloatFromCurve(
       FloatCurve=DataInterface() {propertyValues {Curve {
          defaultValue=1
          keys [
           {arriveTangent=0.074, interpMode=RCIM_Cubic, leaveTangent=0.074, tangentMode=RCTM_User, time=0, value=1}
           {arriveTangent=-8.6402, interpMode=RCIM_Cubic, leaveTangent=-8.6402, tangentMode=RCTM_User, time=0.1915, value=0.6}
           {arriveTangent=0.0063, interpMode=RCIM_Cubic, leaveTangent=0.0063, tangentMode=RCTM_User, time=0.25, value=0}
          ]
         }}}
       CurveIndex=$Engine.System.Age
      )
      2=&FloatFromCurve(
       FloatCurve=DataInterface() {propertyValues {Curve {
          defaultValue=1
          keys [
           {arriveTangent=0.074, interpMode=RCIM_Cubic, leaveTangent=0.074, tangentMode=RCTM_User, time=0, value=1}
           {arriveTangent=-8.6402, interpMode=RCIM_Cubic, leaveTangent=-8.6402, tangentMode=RCTM_User, time=0.1779, value=0.5777}
           {arriveTangent=0.0063, interpMode=RCIM_Cubic, leaveTangent=0.0063, tangentMode=RCTM_User, time=0.25, value=0}
          ]
         }}}
       CurveIndex=$Engine.System.Age
      )
      3=&FloatFromCurve(
       FloatCurve=DataInterface() {propertyValues {Curve {
          defaultValue=1
          keys [
           {arriveTangent=0.074, interpMode=RCIM_Cubic, leaveTangent=0.074, tangentMode=RCTM_User, time=0, value=1}
           {arriveTangent=-11.458, interpMode=RCIM_Cubic, leaveTangent=-11.458, tangentMode=RCTM_User, time=0.1476, value=0.4947}
           {arriveTangent=0.0063, interpMode=RCIM_Cubic, leaveTangent=0.0063, tangentMode=RCTM_User, time=0.2, value=0}
          ]
         }}}
       CurveIndex=$Engine.System.Age
      )
      Index=$User.SwingSectionIndex
      Num=::NewEnumerator2
     )
     B=&FloatFromCurve(
      FloatCurve=DataInterface() {propertyValues {Curve {keys [{interpMode=RCIM_Cubic, tangentMode=RCTM_User, time=0, value=1}, {interpMode=RCIM_Cubic, tangentMode=RCTM_User, time=1, value=0}]}}}
      Scale Curve=$Emitter.rand
     )
    )
    Scale RGB=&MakeVectorFromLinearColorRGB(Color=&ColorFromCurve(ColorCurve=DataInterface() {propertyValues {
        AlphaCurve {keys []}
        BlueCurve {keys [<at 0 1 RCIM_Cubic>,<at 0.6466 0.2479 RCIM_Cubic>]}
        GreenCurve {keys [<at 0 1 RCIM_Cubic>,<at 0.6466 0.875 RCIM_Cubic>]}
        RedCurve {keys [<at 0 1 RCIM_Cubic>,<at 0.6466 0 RCIM_Cubic>]}
       }}))
    ScaleA=true
    ScaleRGB=true
   )
   &DynamicMaterialParameters(
    Index 0 Param 2=&FloatFromCurve(FloatCurve=DataInterface() {propertyValues {Curve {keys [
         {arriveTangent=0.2955, interpMode=RCIM_Cubic, leaveTangent=0.2955, tangentMode=RCTM_User, time=0, value=0}
         {arriveTangent=-2.2131, interpMode=RCIM_Cubic, leaveTangent=-2.2131, tangentMode=RCTM_User, time=1, value=-1}
        ]}}})
    Module.Param0WriteEnabled=false
    Module.Param1WriteEnabled=true
    Module.Param2WriteEnabled=false
    Module.Param3WriteEnabled=false
   )
   /SetVariables(Particles.Trail Color02:LinearColor=&MakeLinearColorFromVectorAndFloat(
     "Vector (RGB)"=&MakeVectorFromLinearColorRGB(Color=&ColorFromCurve(ColorCurve=DataInterface() {propertyValues {
         AlphaCurve {keys [<at 0 1 RCIM_Cubic>,<at 1 0 RCIM_Cubic>]}
         BlueCurve {keys [<at 0 0.1682 RCIM_Cubic>,<at 1 0.75 RCIM_Cubic>]}
         GreenCurve {keys [<at 0 0.5938 RCIM_Cubic>,<at 1 0.1044 RCIM_Cubic>]}
         RedCurve {keys [<at 0 0 RCIM_Cubic>,<at 1 0 RCIM_Cubic>]}
        }}))
     "Float (Alpha)"=&Multiply_Float(
      A=&MergeHelper_FloatSwitch(
       0=&FloatFromCurve(
        FloatCurve=DataInterface() {propertyValues {Curve {
           defaultValue=1
           keys [
            {arriveTangent=0.074, interpMode=RCIM_Cubic, leaveTangent=0.074, tangentMode=RCTM_User, time=0, value=1}
            {arriveTangent=-8.6402, interpMode=RCIM_Cubic, leaveTangent=-8.6402, tangentMode=RCTM_User, time=0.1972, value=0.5421}
            {arriveTangent=0.0063, interpMode=RCIM_Cubic, leaveTangent=0.0063, tangentMode=RCTM_User, time=0.26, value=0}
           ]
          }}}
        CurveIndex=$Engine.System.Age
       )
       1=&FloatFromCurve(
        FloatCurve=DataInterface() {propertyValues {Curve {
           defaultValue=1
           keys [
            {arriveTangent=0.074, interpMode=RCIM_Cubic, leaveTangent=0.074, tangentMode=RCTM_User, time=0, value=1}
            {arriveTangent=-8.6402, interpMode=RCIM_Cubic, leaveTangent=-8.6402, tangentMode=RCTM_User, time=0.1915, value=0.6}
            {arriveTangent=0.0063, interpMode=RCIM_Cubic, leaveTangent=0.0063, tangentMode=RCTM_User, time=0.25, value=0}
           ]
          }}}
        CurveIndex=$Engine.System.Age
       )
       2=&FloatFromCurve(
        FloatCurve=DataInterface() {propertyValues {Curve {
           defaultValue=1
           keys [
            {arriveTangent=0.074, interpMode=RCIM_Cubic, leaveTangent=0.074, tangentMode=RCTM_User, time=0, value=1}
            {arriveTangent=-8.6402, interpMode=RCIM_Cubic, leaveTangent=-8.6402, tangentMode=RCTM_User, time=0.1779, value=0.5777}
            {arriveTangent=0.0063, interpMode=RCIM_Cubic, leaveTangent=0.0063, tangentMode=RCTM_User, time=0.25, value=0}
           ]
          }}}
        CurveIndex=$Engine.System.Age
       )
       3=&FloatFromCurve(
        FloatCurve=DataInterface() {propertyValues {Curve {
           defaultValue=1
           keys [
            {arriveTangent=0.074, interpMode=RCIM_Cubic, leaveTangent=0.074, tangentMode=RCTM_User, time=0, value=1}
            {arriveTangent=-11.458, interpMode=RCIM_Cubic, leaveTangent=-11.458, tangentMode=RCTM_User, time=0.1476, value=0.4947}
            {arriveTangent=0.0063, interpMode=RCIM_Cubic, leaveTangent=0.0063, tangentMode=RCTM_User, time=0.2, value=0}
           ]
          }}}
        CurveIndex=$Engine.System.Age
       )
       Index=$User.SwingSectionIndex
       Num=::NewEnumerator2
      )
      B=&FloatFromCurve(
       FloatCurve=DataInterface() {propertyValues {Curve {keys [
           {interpMode=RCIM_Cubic, tangentMode=RCTM_User, time=0, value=1}
           {arriveTangent=-1.4244, interpMode=RCIM_Cubic, leaveTangent=-1.4244, tangentMode=RCTM_User, time=1, value=0}
          ]}}}
       Scale Curve=$Emitter.rand
      )
     )
    )) {isSetParameters=true}
   /SetVariables(Particles.DY2:Vector4f=&MakeVector4(X=1, Y=1, Z=1, W=1)) {isSetParameters=true}
  ]
  rendererClasses [&NiagaraRibbonRendererProperties]
  renderers [
   {rendererClass=&NiagaraRibbonRendererProperties}
   {rendererClass=&NiagaraRibbonRendererProperties, rendererIndex=1}
  ]
 }
 Sparks {
  emitterUpdateScript [
   &EmitterState()
   &SpawnPerUnit(Module.Use Max Movement Threshold=true, Module.Use Movement Tolerance=false, Spawn Spacing=50, Use Spawn Probability=false)
  ]
  particleSpawnScript [
   &InitializeParticle(
    Color Maximum=<rgba 0.5 5 0.65 1>
    Color Minimum=<rgba 0.4 4 1.42 1>
    Color Mode=::NewEnumerator2
    Lifetime Max=0.4
    Lifetime Min=0.3
    Lifetime Mode=::NewEnumerator1
    Sprite Size Mode=::NewEnumerator1
    Uniform Sprite Size Max=1.5
    Uniform Sprite Size Min=0.8
    UsePositionOffset=false
    Write Material Random=true
   )
   &SkeletalMeshLocation(
    Bone / Socket Index=1
    MaskSpawnGroup=false
    Mesh Sampling Type=::NewEnumerator5
    Skeletal Mesh=$User.SkeletalMesh
    Socket Sampling Mode=::NewEnumerator5
   )
   &InheritVelocity(
    Inherited Velocity Speed Limit:NiagaraFloat=&RandomRangeFloat(Minimum=40, Maximum=80)
    Speed Limit=true
    UseSpeedThreshold=false
   )
   &ApplyInitialForces()
   &ShapeLocation(MaskSpawnGroup=false, Override Seed=false, Sphere Radius=10, UseOwnerScale=true)
  ]
  particleUpdateScript [
   &ParticleState()
   &ScaleSpriteSize(Scale Sprite Size Mode=::NewEnumerator2) {module=ScaleSpriteSize001}
   &ScaleSpriteSize(Uniform Curve Sprite Scale=DataInterface() {propertyValues {Curve {keys [
        {tangentMode=RCTM_User, time=0, value=1}
        {arriveTangent=-0.0405, interpMode=RCIM_Cubic, leaveTangent=-0.0405, tangentMode=RCTM_User, time=0.2, value=1}
        {arriveTangent=-2.1547, interpMode=RCIM_Cubic, leaveTangent=-2.1547, tangentMode=RCTM_User, time=1, value=0}
       ]}}})
   &CurlNoiseForce(Noise Frequency=20, Noise Strength=200, Pan Noise=false)
   &Drag(Drag=&RandomRangeFloat(Minimum=6, Maximum=9), Use Linear Drag=true, Use Rotational Drag=true)
   &SolveForcesAndVelocity(Clamp Velocity=false, Limit Acceleration=false)
   /SetVariables(
    Particles.Color2=&MultiplyLinearColor(
     Linear Color=&RandomRangeLinearColor(Minimum=<rgba 0.0005 0.001 0.0006 1>, Maximum=<rgba 0 0.1 0.0283 1>, Evaluation Type=::NewEnumerator1)
     Scale Factor=&MakeVector4(
      X=&MakeCustomFloatFromBool(Bool=$User.bIsFrontEndPreview, True Float=1, False Float=1)
      Y=&MakeCustomFloatFromBool(Bool=$User.bIsFrontEndPreview, True Float=1, False Float=1)
      Z=&MakeCustomFloatFromBool(Bool=$User.bIsFrontEndPreview, True Float=1, False Float=1)
      W=&FloatFromCurve(
       FloatCurve=DataInterface() {propertyValues {Curve {keys [{interpMode=RCIM_Cubic, tangentMode=RCTM_User, time=0.5, value=1}, {arriveTangent=-6.3011, tangentMode=RCTM_User, time=1, value=0}]}}}
       Scale Curve=0.75
      )
     )
    )
    Particles.SpriteSize2=&Vector2DFromFloat(Value=7.5)
   ) {isSetParameters=true}
   &ScaleColor(
    Scale Alpha=0.5
    Scale RGB=&VectorFromFloat(Value=&Multiply_Float(
      A=&RandomRangeFloat(Minimum=-10, Maximum=10, Evaluation Type=::NewEnumerator1)
      B=&MakeCustomFloatFromBool(Bool=$User.bIsFrontEndPreview, True Float=1, False Float=1)
     ))
    ScaleA=true
    ScaleRGB=true
   )
  ]
  rendererClasses [&NiagaraSpriteRendererProperties]
  renderers [
   {rendererClass=&NiagaraSpriteRendererProperties}
   {rendererClass=&NiagaraSpriteRendererProperties, rendererIndex=1}
  ]
 }
 Trail_FaceCamera {
  emitterUpdateScript [
   &EmitterState()
   &FN_SpawnAnimTrail()
   /SetVariables(Emitter.rand:NiagaraFloat=&RandomRangeFloat(Minimum=1, Maximum=1, Evaluation Type=::NewEnumerator1)) {isSetParameters=true}
  ]
  particleSpawnScript [
   &FN_InitializeAnimTrailBoneIndecies()
   &FN_InitializeAnimTrail(Color=<rgba 1.5 30 13.375 1>, Lifetime=0.2)
   /SetVariables(Particles.Position_Tip:NiagaraPosition=&AddVectorToPosition(
     Position=$Particles.Position
     Vector=&TransformVector(
      Vector=&MergeHelper_VectorSwitch(0=<0 0 12.5>, 1=<0 0 12.5>, 2=<0 0 10>, 3=<0 0 12.5>, Index=$User.SwingSectionIndex, Num=::NewEnumerator2)
      Destination Space=::Simulation
     )
    )) {isSetParameters=true}
  ]
  particleUpdateScript [
   &ParticleState()
   /SetVariables(Particles.Trail width02:NiagaraFloat=&Multiply_Float(A=$Particles.RibbonWidth, B=1)) {isSetParameters=true}
   &FN_UpdateAnimTrail(SkeletalMesh=$User.SkeletalMesh)
   &ScaleColor(
    Scale Alpha=&Multiply_Float(
     A=&MergeHelper_FloatSwitch(
      0=&FloatFromCurve(
       FloatCurve=DataInterface() {propertyValues {Curve {
          defaultValue=1
          keys [
           {arriveTangent=0.074, interpMode=RCIM_Cubic, leaveTangent=0.074, tangentMode=RCTM_User, time=0, value=1}
           {arriveTangent=-8.6402, interpMode=RCIM_Cubic, leaveTangent=-8.6402, tangentMode=RCTM_User, time=0.1972, value=0.5421}
           {arriveTangent=0.0063, interpMode=RCIM_Cubic, leaveTangent=0.0063, tangentMode=RCTM_User, time=0.26, value=0}
          ]
         }}}
       CurveIndex=$Engine.System.Age
      )
      1=&FloatFromCurve(
       FloatCurve=DataInterface() {propertyValues {Curve {
          defaultValue=1
          keys [
           {arriveTangent=0.074, interpMode=RCIM_Cubic, leaveTangent=0.074, tangentMode=RCTM_User, time=0, value=1}
           {arriveTangent=-8.6402, interpMode=RCIM_Cubic, leaveTangent=-8.6402, tangentMode=RCTM_User, time=0.1915, value=0.6}
           {arriveTangent=0.0063, interpMode=RCIM_Cubic, leaveTangent=0.0063, tangentMode=RCTM_User, time=0.25, value=0}
          ]
         }}}
       CurveIndex=$Engine.System.Age
      )
      2=&FloatFromCurve(
       FloatCurve=DataInterface() {propertyValues {Curve {
          defaultValue=1
          keys [
           {arriveTangent=0.074, interpMode=RCIM_Cubic, leaveTangent=0.074, tangentMode=RCTM_User, time=0, value=1}
           {arriveTangent=-8.6402, interpMode=RCIM_Cubic, leaveTangent=-8.6402, tangentMode=RCTM_User, time=0.1779, value=0.5777}
           {arriveTangent=0.0063, interpMode=RCIM_Cubic, leaveTangent=0.0063, tangentMode=RCTM_User, time=0.25, value=0}
          ]
         }}}
       CurveIndex=$Engine.System.Age
      )
      3=&FloatFromCurve(
       FloatCurve=DataInterface() {propertyValues {Curve {
          defaultValue=1
          keys [
           {arriveTangent=0.074, interpMode=RCIM_Cubic, leaveTangent=0.074, tangentMode=RCTM_User, time=0, value=1}
           {arriveTangent=-11.458, interpMode=RCIM_Cubic, leaveTangent=-11.458, tangentMode=RCTM_User, time=0.1476, value=0.4947}
           {arriveTangent=0.0063, interpMode=RCIM_Cubic, leaveTangent=0.0063, tangentMode=RCTM_User, time=0.2, value=0}
          ]
         }}}
       CurveIndex=$Engine.System.Age
      )
      Index=$User.SwingSectionIndex
      Num=::NewEnumerator2
     )
     B=&FloatFromCurve(
      FloatCurve=DataInterface() {propertyValues {Curve {keys [{interpMode=RCIM_Cubic, tangentMode=RCTM_User, time=0, value=1}, {interpMode=RCIM_Cubic, tangentMode=RCTM_User, time=1, value=0}]}}}
      Scale Curve=$Emitter.rand
     )
    )
    Scale RGB=&MakeVectorFromLinearColorRGB(Color=&ColorFromCurve(ColorCurve=DataInterface() {propertyValues {
        AlphaCurve {keys []}
        BlueCurve {keys [<at 0 1 RCIM_Cubic>,<at 0.6466 0.2479 RCIM_Cubic>]}
        GreenCurve {keys [<at 0 1 RCIM_Cubic>,<at 0.6466 0.875 RCIM_Cubic>]}
        RedCurve {keys [<at 0 1 RCIM_Cubic>,<at 0.6466 0 RCIM_Cubic>]}
       }}))
    ScaleA=true
    ScaleRGB=true
   )
   &DynamicMaterialParameters(
    Index 0 Param 1=2.5
    Module.Index0Param0Write=true
    Module.Index0Param1Write=true
    Module.Index0Param2Write=true
    Module.Index0Param3Write=true
   )
   /SetVariables(Particles.DY_Tip:Vector4f=&MakeVector4(
     X=1
     Y=&FloatFromCurve(FloatCurve=DataInterface() {propertyValues {Curve {keys [
          {arriveTangent=0.2955, interpMode=RCIM_Cubic, leaveTangent=0.2955, tangentMode=RCTM_User, time=0, value=0}
          {arriveTangent=-2.2131, interpMode=RCIM_Cubic, leaveTangent=-2.2131, tangentMode=RCTM_User, time=1, value=-1}
         ]}}})
     Z=1
     W=1
    )) {isSetParameters=true}
   /SetVariables(
    Particles.Trail Color Glow:LinearColor=&MakeLinearColorFromVectorAndFloat(
     "Vector (RGB)"=&MakeVectorFromLinearColorRGB(Color=&ColorFromCurve(ColorCurve=DataInterface() {propertyValues {
         AlphaCurve {keys [<at 0 1 RCIM_Cubic>,<at 1 0 RCIM_Cubic>]}
         BlueCurve {keys [<at 0 2.7149 RCIM_Cubic>,<at 1 0.2488 RCIM_Cubic>]}
         GreenCurve {keys [<at 0 25 RCIM_Cubic>,<at 1 0.8783 RCIM_Cubic>]}
         RedCurve {keys [<at 0 0 RCIM_Cubic>,<at 1 0 RCIM_Cubic>]}
        }}))
     "Float (Alpha)"=&Multiply_Float(
      A=&MergeHelper_FloatSwitch(
       0=&FloatFromCurve(
        FloatCurve=DataInterface() {propertyValues {Curve {
           defaultValue=1
           keys [
            {arriveTangent=0.074, interpMode=RCIM_Cubic, leaveTangent=0.074, tangentMode=RCTM_User, time=0, value=1}
            {arriveTangent=-8.6402, interpMode=RCIM_Cubic, leaveTangent=-8.6402, tangentMode=RCTM_User, time=0.1972, value=0.5421}
            {arriveTangent=0.0063, interpMode=RCIM_Cubic, leaveTangent=0.0063, tangentMode=RCTM_User, time=0.26, value=0}
           ]
          }}}
        CurveIndex=$Engine.System.Age
       )
       1=&FloatFromCurve(
        FloatCurve=DataInterface() {propertyValues {Curve {
           defaultValue=1
           keys [
            {arriveTangent=0.074, interpMode=RCIM_Cubic, leaveTangent=0.074, tangentMode=RCTM_User, time=0, value=1}
            {arriveTangent=-8.6402, interpMode=RCIM_Cubic, leaveTangent=-8.6402, tangentMode=RCTM_User, time=0.1915, value=0.6}
            {arriveTangent=0.0063, interpMode=RCIM_Cubic, leaveTangent=0.0063, tangentMode=RCTM_User, time=0.25, value=0}
           ]
          }}}
        CurveIndex=$Engine.System.Age
       )
       2=&FloatFromCurve(
        FloatCurve=DataInterface() {propertyValues {Curve {
           defaultValue=1
           keys [
            {arriveTangent=0.074, interpMode=RCIM_Cubic, leaveTangent=0.074, tangentMode=RCTM_User, time=0, value=1}
            {arriveTangent=-8.6402, interpMode=RCIM_Cubic, leaveTangent=-8.6402, tangentMode=RCTM_User, time=0.1779, value=0.5777}
            {arriveTangent=0.0063, interpMode=RCIM_Cubic, leaveTangent=0.0063, tangentMode=RCTM_User, time=0.25, value=0}
           ]
          }}}
        CurveIndex=$Engine.System.Age
       )
       3=&FloatFromCurve(
        FloatCurve=DataInterface() {propertyValues {Curve {
           defaultValue=1
           keys [
            {arriveTangent=0.074, interpMode=RCIM_Cubic, leaveTangent=0.074, tangentMode=RCTM_User, time=0, value=1}
            {arriveTangent=-11.458, interpMode=RCIM_Cubic, leaveTangent=-11.458, tangentMode=RCTM_User, time=0.1476, value=0.4947}
            {arriveTangent=0.0063, interpMode=RCIM_Cubic, leaveTangent=0.0063, tangentMode=RCTM_User, time=0.2, value=0}
           ]
          }}}
        CurveIndex=$Engine.System.Age
       )
       Index=$User.SwingSectionIndex
       Num=::NewEnumerator2
      )
      B=&FloatFromCurve(
       FloatCurve=DataInterface() {propertyValues {Curve {keys [
           {interpMode=RCIM_Cubic, tangentMode=RCTM_User, time=0, value=1}
           {arriveTangent=-1.4244, interpMode=RCIM_Cubic, leaveTangent=-1.4244, tangentMode=RCTM_User, time=1, value=0}
          ]}}}
       Scale Curve=$Emitter.rand
      )
     )
    )
    Particles.Trail Width Glow=10
   ) {isSetParameters=true}
   /SetVariables(Particles.Trail Color02:LinearColor=&MakeLinearColorFromVectorAndFloat(
     "Vector (RGB)"=&MakeVectorFromLinearColorRGB(Color=&ColorFromCurve(ColorCurve=DataInterface() {propertyValues {
         AlphaCurve {keys [<at 0 1 RCIM_Cubic>,<at 1 0 RCIM_Cubic>]}
         BlueCurve {keys [<at 0 0.6721 RCIM_Cubic>,<at 1 0.8783 RCIM_Cubic>]}
         GreenCurve {keys [<at 0 0.8783 RCIM_Cubic>,<at 1 0.3055 RCIM_Cubic>]}
         RedCurve {keys [<at 0 0 RCIM_Cubic>,<at 1 0 RCIM_Cubic>]}
        }}))
     "Float (Alpha)"=&Multiply_Float(
      A=&MergeHelper_FloatSwitch(
       0=&FloatFromCurve(
        FloatCurve=DataInterface() {propertyValues {Curve {
           defaultValue=1
           keys [
            {arriveTangent=0.074, interpMode=RCIM_Cubic, leaveTangent=0.074, tangentMode=RCTM_User, time=0, value=1}
            {arriveTangent=-8.6402, interpMode=RCIM_Cubic, leaveTangent=-8.6402, tangentMode=RCTM_User, time=0.1972, value=0.5421}
            {arriveTangent=0.0063, interpMode=RCIM_Cubic, leaveTangent=0.0063, tangentMode=RCTM_User, time=0.26, value=0}
           ]
          }}}
        CurveIndex=$Engine.System.Age
       )
       1=&FloatFromCurve(
        FloatCurve=DataInterface() {propertyValues {Curve {
           defaultValue=1
           keys [
            {arriveTangent=0.074, interpMode=RCIM_Cubic, leaveTangent=0.074, tangentMode=RCTM_User, time=0, value=1}
            {arriveTangent=-8.6402, interpMode=RCIM_Cubic, leaveTangent=-8.6402, tangentMode=RCTM_User, time=0.1915, value=0.6}
            {arriveTangent=0.0063, interpMode=RCIM_Cubic, leaveTangent=0.0063, tangentMode=RCTM_User, time=0.25, value=0}
           ]
          }}}
        CurveIndex=$Engine.System.Age
       )
       2=&FloatFromCurve(
        FloatCurve=DataInterface() {propertyValues {Curve {
           defaultValue=1
           keys [
            {arriveTangent=0.074, interpMode=RCIM_Cubic, leaveTangent=0.074, tangentMode=RCTM_User, time=0, value=1}
            {arriveTangent=-8.6402, interpMode=RCIM_Cubic, leaveTangent=-8.6402, tangentMode=RCTM_User, time=0.1779, value=0.5777}
            {arriveTangent=0.0063, interpMode=RCIM_Cubic, leaveTangent=0.0063, tangentMode=RCTM_User, time=0.25, value=0}
           ]
          }}}
        CurveIndex=$Engine.System.Age
       )
       3=&FloatFromCurve(
        FloatCurve=DataInterface() {propertyValues {Curve {
           defaultValue=1
           keys [
            {arriveTangent=0.074, interpMode=RCIM_Cubic, leaveTangent=0.074, tangentMode=RCTM_User, time=0, value=1}
            {arriveTangent=-11.458, interpMode=RCIM_Cubic, leaveTangent=-11.458, tangentMode=RCTM_User, time=0.1476, value=0.4947}
            {arriveTangent=0.0063, interpMode=RCIM_Cubic, leaveTangent=0.0063, tangentMode=RCTM_User, time=0.2, value=0}
           ]
          }}}
        CurveIndex=$Engine.System.Age
       )
       Index=$User.SwingSectionIndex
       Num=::NewEnumerator2
      )
      B=&FloatFromCurve(
       FloatCurve=DataInterface() {propertyValues {Curve {keys [
           {interpMode=RCIM_Cubic, tangentMode=RCTM_User, time=0, value=1}
           {arriveTangent=-1.4244, interpMode=RCIM_Cubic, leaveTangent=-1.4244, tangentMode=RCTM_User, time=1, value=0}
          ]}}}
       Scale Curve=$Emitter.rand
      )
     )
    )) {isSetParameters=true}
   /SetVariables(Particles.DY2:Vector4f=&MakeVector4(X=1, Y=1, Z=1, W=1)) {isSetParameters=true}
  ]
  rendererClasses [&NiagaraRibbonRendererProperties]
  renderers [
   {rendererClass=&NiagaraRibbonRendererProperties}
   {rendererClass=&NiagaraRibbonRendererProperties, rendererIndex=1}
  ]
 }
 Trail_Soft_Dark {
  emitterUpdateScript [
   &EmitterState()
   &FN_SpawnAnimTrail_v2()
  ]
  particleSpawnScript [
   /Game/Effects/Niagara/Modules/AnimTrail/FN_InitializeAnimTrail/FN_InitializeAnimTrailBoneIndecies(Second Bone or Socket Index=1)
   &FN_InitializeAnimTrail(Color=<rgba 0 0.1719 0.0487 1>, Lifetime=0.15) {module=FN_InitializeAnimTrail001}
   /SetVariables(Particles.Position_Tip:NiagaraPosition=&AddVectorToPosition(Position=$Particles.Position, Vector=&TransformVector(Vector=<0 0 12.5>, Destination Space=::Simulation))) {isSetParameters=true}
  ]
  particleUpdateScript [
   &FN_UpdateAnimTrail(SkeletalMesh=$User.SkeletalMesh)
   &ScaleRibbonWidth(Ribbon Width Scale=&FloatFromCurve(FloatCurve=DataInterface() {propertyValues {Curve {keys [
         {interpMode=RCIM_Cubic, tangentMode=RCTM_User, time=0, value=1.1}
         {arriveTangent=-2.3701, interpMode=RCIM_Cubic, leaveTangent=-2.3701, tangentMode=RCTM_User, time=1, value=0}
        ]}}}))
   &ScaleColor(
    Scale Alpha=&Multiply_Float(
     A=&MergeHelper_FloatSwitch(
      0=&FloatFromCurve(
       FloatCurve=DataInterface() {propertyValues {Curve {
          defaultValue=1
          keys [
           {arriveTangent=0.074, interpMode=RCIM_Cubic, leaveTangent=0.074, tangentMode=RCTM_User, time=0, value=1}
           {arriveTangent=-8.6402, interpMode=RCIM_Cubic, leaveTangent=-8.6402, tangentMode=RCTM_User, time=0.1972, value=0.5421}
           {arriveTangent=0.0063, interpMode=RCIM_Cubic, leaveTangent=0.0063, tangentMode=RCTM_User, time=0.26, value=0}
          ]
         }}}
       CurveIndex=$Engine.System.Age
      )
      1=&FloatFromCurve(
       FloatCurve=DataInterface() {propertyValues {Curve {
          defaultValue=1
          keys [
           {arriveTangent=0.074, interpMode=RCIM_Cubic, leaveTangent=0.074, tangentMode=RCTM_User, time=0, value=1}
           {arriveTangent=-8.6402, interpMode=RCIM_Cubic, leaveTangent=-8.6402, tangentMode=RCTM_User, time=0.1915, value=0.6}
           {arriveTangent=0.0063, interpMode=RCIM_Cubic, leaveTangent=0.0063, tangentMode=RCTM_User, time=0.25, value=0}
          ]
         }}}
       CurveIndex=$Engine.System.Age
      )
      2=&FloatFromCurve(
       FloatCurve=DataInterface() {propertyValues {Curve {
          defaultValue=1
          keys [
           {arriveTangent=0.074, interpMode=RCIM_Cubic, leaveTangent=0.074, tangentMode=RCTM_User, time=0, value=1}
           {arriveTangent=-8.6402, interpMode=RCIM_Cubic, leaveTangent=-8.6402, tangentMode=RCTM_User, time=0.1779, value=0.5777}
           {arriveTangent=0.0063, interpMode=RCIM_Cubic, leaveTangent=0.0063, tangentMode=RCTM_User, time=0.25, value=0}
          ]
         }}}
       CurveIndex=$Engine.System.Age
      )
      3=&FloatFromCurve(
       FloatCurve=DataInterface() {propertyValues {Curve {
          defaultValue=1
          keys [
           {arriveTangent=0.074, interpMode=RCIM_Cubic, leaveTangent=0.074, tangentMode=RCTM_User, time=0, value=1}
           {arriveTangent=-11.458, interpMode=RCIM_Cubic, leaveTangent=-11.458, tangentMode=RCTM_User, time=0.1476, value=0.4947}
           {arriveTangent=0.0063, interpMode=RCIM_Cubic, leaveTangent=0.0063, tangentMode=RCTM_User, time=0.2, value=0}
          ]
         }}}
       CurveIndex=$Engine.System.Age
      )
      Index=$User.SwingSectionIndex
      Num=::NewEnumerator2
     )
     B=&FloatFromCurve(
      FloatCurve=DataInterface() {propertyValues {Curve {keys [<at 0 1 RCIM_Cubic>, {arriveTangent=-1.9571, interpMode=RCIM_Cubic, leaveTangent=-1.9571, tangentMode=RCTM_User, time=1, value=0}]}}}
      Scale Curve=&FloatFromCurve(
       FloatCurve=DataInterface() {propertyValues {Curve {keys [
           {interpMode=RCIM_Cubic, tangentMode=RCTM_User, time=0.1, value=0}
           <at 0.3 1>
           {interpMode=RCIM_Cubic, tangentMode=RCTM_User, time=1, value=1}
          ]}}}
       Scale Curve=0.8
      )
     )
    )
    Scale RGB=&Multiply_VectorByFloat(
     Vector=&MakeVectorFromLinearColorRGB(Color=&ColorFromCurve(ColorCurve=DataInterface() {propertyValues {
         AlphaCurve {keys [<at 0 1 RCIM_Cubic>,<at 1 0 RCIM_Cubic>]}
         BlueCurve {keys [<at 0 1 RCIM_Cubic>,<at 1 1 RCIM_Cubic>]}
         GreenCurve {keys [<at 0 1 RCIM_Cubic>,<at 1 0.3493 RCIM_Cubic>]}
         RedCurve {keys [<at 0 1 RCIM_Cubic>,<at 1 0.1324 RCIM_Cubic>]}
        }}))
     Float=$System.TOD
    )
    ScaleA=true
    ScaleRGB=true
   ) {module=ScaleColor001}
   &SpriteFacingAndAlignment(Sprite Facing=$Particles.RibbonFacing, WriteAlignment=false, WriteFacing=true)
   &DynamicMaterialParameters(
    Index 0 Param 1=&FloatFromCurve(FloatCurve=DataInterface() {propertyValues {Curve {keys [<at 0 1 RCIM_Cubic>, {arriveTangent=-1, interpMode=RCIM_Cubic, leaveTangent=-1, time=0.5, value=0.8}, <at 1 0>]}}})
    Module.Param0WriteEnabled=true
    Module.Param1WriteEnabled=true
    Module.Param2WriteEnabled=true
    Module.Param3WriteEnabled=true
   )
  ]
  rendererClasses [&NiagaraRibbonRendererProperties]
  renderers [{rendererClass=&NiagaraRibbonRendererProperties}]
 }
}
summary {
 systemName=NS_Pickaxe_SatireCane_Ode_Trail
 emitters [
  {emitterName=Sparks, rendererClasses [&NiagaraSpriteRendererProperties]}
  {emitterName=Trail_Soft_Dark, rendererClasses [&NiagaraRibbonRendererProperties]}
  {emitterName=Trail_FaceCamera, rendererClasses [&NiagaraRibbonRendererProperties]}
  {emitterName=Lightning, rendererClasses [&NiagaraRibbonRendererProperties]}
 ]
 userVariables [
  User.bIsFrontEnd {type=NiagaraBool}
  User.bIsFrontEndPreview {type=NiagaraBool}
  User.RibbonIDOffset {type=NiagaraInt32}
  User.SkeletalMesh {
   type=NiagaraDataInterfaceSkeletalMesh
   defaultValue {
    struct=NiagaraExt_VariableValue_DataInterface
    value {dataInterface=&self:NiagaraDataInterfaceSkeletalMesh_2}
   }
  }
  User.SwingSectionIndex {type=NiagaraInt32}
 ]
}
